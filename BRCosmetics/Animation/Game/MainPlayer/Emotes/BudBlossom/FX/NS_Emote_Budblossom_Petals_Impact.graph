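# Niagara effect for the BudBlossom emote that bursts a small cluster of sprite petals from a sphere.
# Gravity, curl noise, wind, vortex and aerodynamic drag make them flutter and fall, with sprite facing driven by mesh orientation.
# Collision, gated by platform quality or front-end preview, aligns landed petals to the surface.
# A BlendSpace module switches world/local position blending on owner speed; the SmokeLWE emitter has no renderers.
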NiagaraSystem /BRCosmetics/Animation/Game/MainPlayer/Emotes/BudBlossom/FX/NS_Emote_Budblossom_Petals_Impact
dependencies {
 usedDataInterfaces [&NiagaraDataInterfaceCurve, &NiagaraDataInterfacePlatformSet, &NiagaraDataInterfaceSocketReader]
 usedDynamicInputs [&MergeHelper_FloatSwitch, &MakeIntFromFloat]
 usedModules [
  &SystemState
  &BlendSpaceHelper
  &self:SystemSpawnScript.SystemScriptSource.SystemScriptGraph.NiagaraNodeAssignment_0.SetVariables_#1
  &self:SystemSpawnScript.SystemScriptSource.SystemScriptGraph.NiagaraNodeAssignment_1.SetVariables_#2
  &self:SystemSpawnScript.SystemScriptSource.SystemScriptGraph.NiagaraNodeAssignment_2.SetVariables_#3
  &self:SystemSpawnScript.SystemScriptSource.SystemScriptGraph.NiagaraNodeAssignment_3.SetVariables_#4
 ]
}
emitters {
 Petals_Impact {
  emitterUpdateScript [
   &EmitterState(Life Cycle Mode=::NewEnumerator1, Loop Behavior=::NewEnumerator1, Loop Delay=4.2, Loop Duration=1, UseLoopDelay=false)
   &SpawnBurst_Instantaneous(Spawn Count=11, Use Spawn Probability=false, UseLoopCountLimit=false)
  ]
  particleSpawnScript [
   &InitializeParticle(
    Lifetime Max=1.5
    Lifetime Min=0.5
    Lifetime Mode=::NewEnumerator1
    Mass Max=1.5
    Mass Mode=::NewEnumerator3
    Position Offset=<-40.9585 0 63.7918>
    Sprite Rotation Mode=::NewEnumerator2
    Sprite Size Mode=::NewEnumerator1
    Uniform Sprite Size Max=3.5
    Uniform Sprite Size Min=2
    UsePositionOffset=false
    Write Material Random=true
   )
   &ShapeLocation(MaskSpawnGroup=false, Override Seed=false, Sphere Radius=4, UseOwnerScale=true)
   &InitialMeshOrientation(Use Rotation Vector=false)
   &ApplyInitialForces()
   &AddVelocity(Velocity=&RandomRangeVector(Minimum=<-100 -100 0>, Maximum=<100 100 0>)) {module=AddVelocity001}
  ]
  particleUpdateScript [
   &ParticleState()
   &GravityForce(Gravity=&RandomRangeVector(Minimum=<0 0 -900>, Maximum=<0 0 -500>))
   &CurlNoiseForce(Bias Noise=false, Noise Frequency=100, Noise Strength=1000, Pan Noise=false)
   &AddRotationalVelocity(Rotation Rate=&RandomRangeVector(Minimum=<0 -0.1 -0.1>, Maximum=<0 0.12 0.12>))
   &WindForce(
    Offset By Time:Vector3f=<50 10 -125>
    Scale Min / Max:Vector2f=<0.275 1>
    Speed Min / Max:Vector2f=<0 500>
    Turbulence Scale Mode:ENiagaraWindTurbulenceContributionMode=::NewEnumerator1
    Wind Speed=&RandomRangeVector(Minimum=<-50 -150 0>, Maximum=<50 150 1>)
    Wind Speed Scale=&Multiply_Float(
     A=&RandomRangeFloat(Minimum=0.3, Maximum=0.6)
     B=&FloatFromCurve(FloatCurve=DataInterface() {propertyValues {Curve {keys [<at 0.9 1 RCIM_Cubic>, {arriveTangent=-12.3129, interpMode=RCIM_Cubic, leaveTangent=-12.3129, tangentMode=RCTM_User, time=1, value=0}]}}})
    )
   )
   &VortexForce(UseOriginOffset=false, Vortex Force Amount=250)
   &AerodynamicDrag(
    Aerodynamic Drag:NiagaraFloat=&RandomRangeFloat(Minimum=3, Maximum=8)
    Aerodynamic Rotational Drag=0.05
    Lift Contribution=0.75
    Pivot Offset Maximum:Vector3f=<5 0.5 1>
    Pivot Offset Minimum:Vector3f=<-5 -0.5 -1>
    Rotation Strength:NiagaraFloat=&RandomRangeFloat(Minimum=-3, Maximum=3)
    Use Linear Drag=true
    Use Rotational Drag=true
   )
   &Collision(
    Age Colliding Particles=false
    Collision Enabled=/Niagara/DynamicInputs/Helpers/MaskBoolBySpawnGroup(
     Masked Bool:NiagaraBool=&Boolean_OrOperation(
      Bool1=&Boolean_OrOperation(
       Bool1=&IsPlatformSetActive(New Platform Set=DataInterface() {propertyValues {Platforms {qualityLevelMask=-8}}})
       Bool2=$User.bIsFrontEndPreview
      )
      Bool2=false
     )
     Passthrough Bool:NiagaraBool=true
     Spawn Group Mask:NiagaraInt32=0
    )
    Friction=2
    Friction During a Bounce=0.1
    KillOnCollision=false
    Module.Control Rest State Via Surface Normal=false
    Module.EnableMaxCollisionCount=false
    Module.Randomize Collision Normal Vector=false
    Radius Calculation Type=::NewEnumerator1
    Restitution=0.02
   ) {module=Collision002}
   /Niagara/Modules/Collision/AlignParticlesWithCollisionPlane(Allow Alignment on Y=false, Allow Alignment on Z=false, Force Alignment:NiagaraBool=$Particles.HasCollided, Rotational Dampening Power=1)
   &SolveForcesAndVelocity(Clamp Velocity=false, Limit Acceleration=false)
   &ScaleSpriteSize(
    Uniform Curve Scale=&FloatFromCurve(FloatCurve=DataInterface() {propertyValues {Curve {keys [<at 0.85 1>,<at 1 0>]}}}, CurveIndex=$Emitter.NormalizedLoopAge)
    Uniform Curve Sprite Scale=DataInterface() {propertyValues {Curve {keys [
        {arriveTangent=14.5835, interpMode=RCIM_Cubic, leaveTangent=14.5835, tangentMode=RCTM_User, time=0, value=0}
        {arriveTangent=1.1111, interpMode=RCIM_Cubic, leaveTangent=1.1111, time=0.1, value=1}
        {arriveTangent=-1.1111, interpMode=RCIM_Cubic, leaveTangent=-1.1111, time=0.9, value=1}
        {arriveTangent=-8.7113, interpMode=RCIM_Cubic, leaveTangent=-8.7113, tangentMode=RCTM_User, time=1, value=0}
       ]}}}
   )
   &AlignSpriteToMeshOrientation(SetAlignment=true, SetFacing=true)
   &BlendSpace("Position Blend (World <-> Local)":NiagaraFloat=&MakeCustomFloatFromBool(
     Bool=&SetBoolByFloatComparison(A=&VectorLength(Vector=$Engine.Owner.Velocity), B=600)
     True Float=1
     False Float=0
     Translation Contribution:NiagaraFloat=0.9
     Rotation Contribution:NiagaraFloat=0
    ))
  ]
  rendererClasses [&NiagaraSpriteRendererProperties]
  renderers [{rendererClass=&NiagaraSpriteRendererProperties}]
 }
 SmokeLWE {rendererClasses [], renderers []}
}
summary {
 systemName=NS_Emote_Budblossom_Petals_Impact
 emitters [{emitterName=Petals_Impact, rendererClasses [&NiagaraSpriteRendererProperties]}, {emitterName=SmokeLWE, rendererClasses []}]
 userVariables [
  User.bIsFrontEnd {type=NiagaraBool}
  User.bIsFrontEndPreview {type=NiagaraBool}
  User.IsSkydiving {type=NiagaraBool}
  User.OwnerLocalVelocity {type=Vector3f, description="", defaultValue {struct=Vector3f, value=<0 0 0>}}
  User.Variant {type=NiagaraFloat, defaultValue=0, description=""}
  User.VariantColorIndex {type=NiagaraInt32}
 ]
}
