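# Niagara system for the BudBlossom emote, emitting looping sprite petals from the gadget's skeletal mesh.
# Spawning combines distance-based and rate spawners, switched by User.Intro, with an extra quality-gated intro ramp; User.hand selects lifetime curves timed to loop age.
# Petals flutter through curl noise, gravity, wind, vortex and aerodynamic drag, with platform- or preview-gated collision.
# BlendSpace blends world and local position based on owner speed.
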NiagaraSystem /BRCosmetics/Animation/Game/MainPlayer/Emotes/BudBlossom/FX/NS_Emote_Budblossom_Petals_Loop
dependencies {
 usedDataInterfaces [
  &NiagaraDataInterfaceCurve
  &NiagaraDataInterfacePlatformSet
  &NiagaraDataInterfaceSkeletalMesh
  &NiagaraDataInterfaceSocketReader
 ]
 usedDynamicInputs []
 usedModules [
  &SystemState
  &BlendSpaceHelper
  &self:SystemSpawnScript.SystemScriptSource.SystemScriptGraph.NiagaraNodeAssignment_0.SetVariables_#1
  &self:SystemSpawnScript.SystemScriptSource.SystemScriptGraph.NiagaraNodeAssignment_1.SetVariables_#2
  &self:SystemSpawnScript.SystemScriptSource.SystemScriptGraph.NiagaraNodeAssignment_2.SetVariables_#3
  &self:SystemSpawnScript.SystemScriptSource.SystemScriptGraph.NiagaraNodeAssignment_3.SetVariables_#4
 ]
}
emitters {Petals {
  emitterUpdateScript [
   &EmitterState(Life Cycle Mode=::NewEnumerator1, Loop Behavior=::NewEnumerator1, Loop Duration Mode=::NewEnumerator1, UseLoopDelay=false)
   &SpawnPerUnit(
    Module.Use Max Movement Threshold=false
    Module.Use Movement Tolerance=true
    Movement Tolerance=10
    Spawn Spacing:NiagaraFloat=&Multiply_Float(
     A=17
     B=&MergeHelper_FloatSwitch(
      0=1.3
      1=0.465
      2=?
      3=?
      Index=&MakeIntFromFloat(Float=$User.Intro, Conversion Mode=::NewEnumerator1)
      Num=::NewEnumerator0
     )
    )
    Use Spawn Probability=false
   )
   &SpawnRate(
    SpawnRate=&Multiply_Float(
     A=&RandomRangeFloat(Minimum=11, Maximum=19)
     B=&FloatFromCurve(
      FloatCurve=DataInterface() {propertyValues {Curve {
         defaultValue=1
         keys [<at 0.8738 1 RCIM_Cubic>, {arriveTangent=74.7011, interpMode=RCIM_Cubic, leaveTangent=74.7011, tangentMode=RCTM_User, time=0.9647, value=3}]
        }}}
      CurveIndex=$Emitter.NormalizedLoopAge
      Scale Curve=&MergeHelper_FloatSwitch(
       0=0.75
       1=1.5
       2=?
       3=?
       Index=&MakeIntFromFloat(Float=$User.Intro, Conversion Mode=::NewEnumerator1)
       Num=::NewEnumerator0
      )
     )
    )
    Use Spawn Probability=false
   )
   &SpawnRate(
    SpawnRate=&MergeHelper_FloatSwitch(
     0=0
     1=&MakeCustomFloatFromBool(
      Bool=&IsPlatformSetActive(New Platform Set=DataInterface() {propertyValues {Platforms {qualityLevelMask=-4}}})
      True Float=&FloatFromCurve(
       FloatCurve=DataInterface() {propertyValues {Curve {
          defaultValue=0
          keys [
           {interpMode=RCIM_Cubic, tangentMode=RCTM_User, time=0, value=0}
           {arriveTangent=6.6699, interpMode=RCIM_Cubic, leaveTangent=6.6699, tangentMode=RCTM_User, time=0.3, value=1}
          ]
         }}}
       CurveIndex=$Emitter.LoopedAge
       Scale Curve=100
      )
      False Float=0
     )
     2=?
     3=?
     Index=&MakeIntFromFloat(Float=$User.Intro, Conversion Mode=::NewEnumerator1)
     Num=::NewEnumerator0
    )
    Use Spawn Probability=false
   ) {module=SpawnRate001}
  ]
  particleSpawnScript [
   &InitializeParticle(
    Lifetime=&Multiply_Float(
     A=&RandomRangeFloat(Minimum=0.5, Maximum=1.1)
     B=&MergeHelper_FloatSwitch(
      0=&FloatFromCurve(
       FloatCurve=DataInterface() {propertyValues {Curve {keys [
           <at 1.2329 1 RCIM_Cubic>
           {arriveTangent=-1.5342, interpMode=RCIM_Cubic, leaveTangent=-0.0059, tangentMode=RCTM_Break, time=1.3392, value=0.4}
           {interpMode=RCIM_Cubic, leaveTangent=6.1355, tangentMode=RCTM_Break, time=3.2651, value=0.4}
           {arriveTangent=1.4098, interpMode=RCIM_Cubic, leaveTangent=1.4098, time=3.3588, value=1}
           <at 4.4 2 RCIM_Cubic>
          ]}}}
       CurveIndex=$Emitter.LoopedAge
      )
      1=&FloatFromCurve(
       FloatCurve=DataInterface() {propertyValues {Curve {keys [
           <at 1.8855 1 RCIM_Cubic>
           {arriveTangent=-1.5342, interpMode=RCIM_Cubic, leaveTangent=-0.0059, tangentMode=RCTM_Break, time=1.9855, value=0.4}
           {interpMode=RCIM_Cubic, leaveTangent=6.1355, tangentMode=RCTM_Break, time=3.4153, value=0.4}
           {arriveTangent=3.3012, interpMode=RCIM_Cubic, leaveTangent=3.3012, time=3.5153, value=1.5}
           <at 3.9 2 RCIM_Cubic>
          ]}}}
       CurveIndex=$Emitter.LoopedAge
      )
      2=&FloatFromCurve(
       FloatCurve=DataInterface() {propertyValues {Curve {keys [
           <at 1.9786 1 RCIM_Cubic>
           {arriveTangent=-1.5342, interpMode=RCIM_Cubic, leaveTangent=-0.0059, tangentMode=RCTM_Break, time=2.0786, value=0.3364}
           {interpMode=RCIM_Cubic, leaveTangent=6.1355, tangentMode=RCTM_Break, time=3.6474, value=0.3364}
           <at 3.7474 1 RCIM_Cubic>
          ]}}}
       CurveIndex=$Emitter.LoopedAge
       Scale Curve=0
      )
      3=?
      Index=&MakeIntFromFloat(Float=$User.hand, Conversion Mode=::NewEnumerator1)
      Num=::NewEnumerator0
     )
    )
    Mass Max=1.5
    Mass Mode=::NewEnumerator3
    Position Offset=<-40.9585 0 63.7918>
    Sprite Rotation Mode=::NewEnumerator2
    Sprite Size Mode=::NewEnumerator1
    Uniform Sprite Size Max=3.5
    Uniform Sprite Size Min=2
    UsePositionOffset=false
    Write Material Random=true
   )
   &SkeletalMeshLocation(
    Bone Sampling Mode=::NewEnumerator4
    MaskSpawnGroup=false
    Skeletal Mesh=DataInterface() {propertyValues {
      DefaultMesh=null
      ExcludeBoneName=root
      OptionalComponentClass=null
      PreviewMesh=/BRCosmetics/Animation/Props/Gadget_BudBlossom/Meshes/BudBlossom_Gadget
      SkinningMode=SkinOnTheFly
      SoftSourceActor=null
      SourceMode=Default
      UvSetIndex=0
      WholeMeshLOD=-1
      bExcludeBone=false
      bReadDeformedGeometry=true
      bRequireCurrentFrameData=true
      ComponentTags []
      FilteredBones [JNT_Stem_5]
      FilteredSockets []
      MeshUserParameter {parameter {name=None, typeDefHandle {registeredTypeIndex=1}, varData []}}
      SamplingRegions []
     }}
   )
   &ShapeLocation(MaskSpawnGroup=false, Offset=<0 0 5>, Offset Mode=::NewEnumerator0, Override Seed=false, Sphere Radius=5, UseOwnerScale=true)
   &InitialMeshOrientation(Use Rotation Vector=false)
   &InheritVelocity(Inherited Velocity Amount Scale=<0.15 0.15 0.15>, Speed Limit=false, UseSpeedThreshold=false)
   &ApplyInitialForces()
   &AddVelocity(Module.Use Falloff Distance=false, Override Seed=false, Velocity Mode=::NewEnumerator1)
  ]
  particleUpdateScript [
   &ParticleState()
   &CurlNoiseForce(Bias Noise=false, Noise Frequency=100, Noise Strength=1000, Pan Noise=false)
   &GravityForce(Gravity=&RandomRangeVector(Minimum=<0 0 -750>, Maximum=<0 0 -350>))
   &AddRotationalVelocity(Rotation Rate=&RandomRangeVector(Minimum=<0 -0.1 -0.1>, Maximum=<0 0.12 0.12>))
   &WindForce(
    Offset By Time:Vector3f=<50 10 -125>
    Scale Min / Max:Vector2f=<0.275 1>
    Speed Min / Max:Vector2f=<0 500>
    Turbulence Scale Mode:ENiagaraWindTurbulenceContributionMode=::NewEnumerator1
    Wind Speed=&RandomRangeVector(Minimum=<-50 -150 0>, Maximum=<50 150 1>)
    Wind Speed Scale=&Multiply_Float(
     A=&RandomRangeFloat(Minimum=0.1, Maximum=0.2)
     B=&FloatFromCurve(FloatCurve=DataInterface() {propertyValues {Curve {keys [<at 0.9 1 RCIM_Cubic>, {arriveTangent=-12.3129, interpMode=RCIM_Cubic, leaveTangent=-12.3129, tangentMode=RCTM_User, time=1, value=0}]}}})
    )
   )
   &VortexForce(UseOriginOffset=false)
   &AerodynamicDrag(
    Aerodynamic Drag:NiagaraFloat=&RandomRangeFloat(Minimum=0.65, Maximum=1.5)
    Aerodynamic Rotational Drag=0.15
    Lift Contribution=0.75
    Pivot Offset Maximum:Vector3f=<5 0.5 1>
    Pivot Offset Minimum:Vector3f=<-5 -0.5 -1>
    Rotation Strength:NiagaraFloat=&RandomRangeFloat(Minimum=-3, Maximum=3)
    Use Linear Drag=true
    Use Rotational Drag=true
   )
   &Collision(
    Age Colliding Particles=false
    Collision Enabled=/Niagara/DynamicInputs/Helpers/MaskBoolBySpawnGroup(
     Masked Bool:NiagaraBool=&Boolean_OrOperation(
      Bool1=&Boolean_OrOperation(
       Bool1=&IsPlatformSetActive(New Platform Set=DataInterface() {propertyValues {Platforms {qualityLevelMask=-8}}})
       Bool2=$User.bIsFrontEndPreview
      )
      Bool2=false
     )
     Passthrough Bool:NiagaraBool=true
     Spawn Group Mask:NiagaraInt32=0
    )
    Friction=2
    Friction During a Bounce=0.1
    KillOnCollision=false
    Module.Control Rest State Via Surface Normal=false
    Module.EnableMaxCollisionCount=false
    Module.Randomize Collision Normal Vector=false
    Radius Calculation Type=::NewEnumerator1
    Restitution=0.02
   ) {module=Collision002}
   /Niagara/Modules/Collision/AlignParticlesWithCollisionPlane(Allow Alignment on Y=false, Allow Alignment on Z=false, Force Alignment:NiagaraBool=$Particles.HasCollided, Rotational Dampening Power=1)
   &SolveForcesAndVelocity(Clamp Velocity=false, Limit Acceleration=false)
   &ScaleSpriteSize(Uniform Curve Sprite Scale=DataInterface() {propertyValues {Curve {keys [
        {arriveTangent=14.5835, interpMode=RCIM_Cubic, leaveTangent=14.5835, tangentMode=RCTM_User, time=0, value=0}
        {arriveTangent=1.1111, interpMode=RCIM_Cubic, leaveTangent=1.1111, time=0.1, value=1}
        {arriveTangent=-1.1111, interpMode=RCIM_Cubic, leaveTangent=-1.1111, time=0.9, value=1}
        {arriveTangent=-8.7113, interpMode=RCIM_Cubic, leaveTangent=-8.7113, tangentMode=RCTM_User, time=1, value=0}
       ]}}})
   &AlignSpriteToMeshOrientation(SetAlignment=true, SetFacing=true)
   &BlendSpace("Position Blend (World <-> Local)":NiagaraFloat=&MakeCustomFloatFromBool(
     Bool=&SetBoolByFloatComparison(A=&VectorLength(Vector=$Engine.Owner.Velocity), B=600)
     True Float=1
     False Float=0
     Translation Contribution:NiagaraFloat=0.9
     Rotation Contribution:NiagaraFloat=0
    ))
  ]
  rendererClasses [&NiagaraSpriteRendererProperties]
  renderers [{rendererClass=&NiagaraSpriteRendererProperties}]
 }}
summary {
 systemName=NS_Emote_Budblossom_Petals_Loop
 emitters [{emitterName=Petals, rendererClasses [&NiagaraSpriteRendererProperties]}]
 userVariables [
  User.bIsFrontEnd {type=NiagaraBool}
  User.bIsFrontEndPreview {type=NiagaraBool}
  User.hand {type=NiagaraFloat, defaultValue=0, description=""}
  User.Intro {type=NiagaraFloat, defaultValue=0, description=""}
  User.IsSkydiving {type=NiagaraBool}
  User.OwnerLocalVelocity {type=Vector3f, description="", defaultValue {struct=Vector3f, value=<0 0 0>}}
  User.VariantColorIndex {type=NiagaraInt32}
 ]
}
